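# Niagara system for the MistTycoonLife character's boots, rendering audio-reactive pink fire, sparks and sparkles.
# Spawn groups map particles to sockets through a shared socket reader, with per-group arrays setting mesh scale, orientation and spark direction.
# System AudioReactivity drives flame scale, material parameters, opacity, spark rate and speed, alongside a fake audio loop module.
# Front-end preview uses timeline-scaled fades and disables movement-based sparks; sparkle brightness varies by quality level.
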NiagaraSystem /BRCosmetics/Characters/Player/Female/Medium/Bodies/F_MED_MistTycoonLife/FX/NS_Character_Body_MistTycoonLife_Boots
dependencies {
 usedDataInterfaces [
  &NiagaraDataInterfaceArrayFloat3
  &NiagaraDataInterfaceArrayInt32
  &NiagaraDataInterfaceCurve
  &NiagaraDataInterfaceArrayFloat
  &NiagaraDataInterfaceColorCurve
  &NiagaraDataInterfaceSocketReader
 ]
 usedDynamicInputs [
  &MakeFloatFromVector2d
  &InvertBool
  /Niagara/DynamicInputs/Lerp/SmoothLerpOverTime_Float
  &RemapRange
  &Cosine
  &Divide_Float
  &MakeFloatFromLinearColor
 ]
 usedModules [
  &self:SystemSpawnScript.SystemScriptSource.SystemScriptGraph.NiagaraNodeAssignment_0.SetVariables_#1
  &self:SystemSpawnScript.SystemScriptSource.SystemScriptGraph.NiagaraNodeAssignment_1.SetVariables_#2
  &self:SystemSpawnScript.SystemScriptSource.SystemScriptGraph.NiagaraNodeAssignment_4.SetVariables_#3
  &SystemState
  /Game/Effects/Niagara/Modules/Audio/FN_FakeAudioReactivityLoop
  &self:SystemSpawnScript.SystemScriptSource.SystemScriptGraph.NiagaraNodeAssignment_3.SetVariables_#4
  &self:SystemSpawnScript.SystemScriptSource.SystemScriptGraph.NiagaraNodeAssignment_2.SetVariables_#5
  &self:SystemSpawnScript.SystemScriptSource.SystemScriptGraph.NiagaraNodeAssignment_6.SetVariables_#6
  &Timeline
 ]
}
emitters {
 Fire_Mesh {
  emitterUpdateScript [
   &EmitterState(
    Inactive Response=::NewEnumerator1
    Life Cycle Mode=::NewEnumerator1
    Loop Behavior=::NewEnumerator1
    Loop Duration Mode=::NewEnumerator1
    UseLoopDelay=false
   )
   &SpawnBurst_Instantaneous(Spawn Count=6, Use Spawn Probability=false, UseLoopCountLimit=false)
  ]
  particleSpawnScript [
   &FN_SetSpawnGroup(Number of Spawn Groups=6)
   &InitializeParticle(
    Color=<rgba 10 5.469 7.9523 1>
    Lifetime=0.1
    Mesh Scale=&SelectVectorFromArray(
     Direct Array Index=&SpawnGroup()
     Vector Selection Array=DataInterface() {propertyValues {FloatData [<2 2 1.5>, <2 2 1.25>, <2 2 1>, <2 2 1.5>, <2 2 1.25>, <2 2 1>]}}
    )
    Mesh Scale Mode=::NewEnumerator4
    UsePositionOffset=false
    Write Material Random=true
   )
   &InitialMeshOrientation(
    Mesh Axis To Orient=<0 0 1>
    Mesh Orientation Mode=::NewEnumerator2
    Orientation Coordinate Space=::Local
    Orientation Vector=<0 1 0>
    Rotation=&SelectVectorFromArray(
     Direct Array Index=&SpawnGroup()
     Vector Selection Array=DataInterface() {propertyValues {FloatData [<0.045 0 0.09>, <0.045 0 0.09>, <0.045 0 0.09>, <0.2432 -0.1261 0>, <0.2432 -0.1261 0>, <0.2432 -0.1261 0>]}}
    )
    Rotation Coordinate Space=::NewEnumerator2
    Use Rotation Vector=true
   ) {module=InitialMeshOrientation001}
  ]
  particleUpdateScript [
   &ParticleState(Kill Particles When Lifetime Has Elapsed=false, Let Infinitely Lived Particles Die When Emitter Deactivates=true)
   &SolveForcesAndVelocity(Clamp Velocity=false, Limit Acceleration=false)
   &ScaleMeshSize(Scale Factor=&Lerp_Vector(A=<0.25 0.75 0.75>, B=<0.25 1 0.75>, Alpha=$System.AudioReactivity))
   &DynamicMaterialParameters(
    Index 0 Param 1=&Lerp_Float(A=1, B=0.25, Alpha=$System.AudioReactivity)
    Index 0 Param 2=0.25
    Module.Index0Float0Write=true
    Module.Index0Float1Write=true
    Module.Index0Float2Write=true
    Module.Index0Float3Write=true
   ) {module=DynamicMaterialParameters002}
   &SocketLocation(
    SamplePosOffset=false
    Socket Index=&SelectIntFromArray(Direct Array Index=&SpawnGroup(), Int Selection Array=DataInterface() {propertyValues {IntData [0,1,2,3,4,5]}})
    Socket Reader=$System.Socket Reader
    Socket Sampling Mode=::NewEnumerator3
    VelocitySpeedLimit=false
   )
   &ScaleColor(
    Scale Alpha=&FloatFromCurve(
     FloatCurve=DataInterface() {propertyValues {Curve {defaultValue=1, keys [{interpMode=RCIM_Cubic, tangentMode=RCTM_Break, time=0, value=0}, {tangentMode=RCTM_User, time=1, value=2}]}}}
     CurveIndex=$System.AudioReactivity
     Scale Curve=&MakeCustomFloatFromBool(Bool=$User.bIsFrontEndPreview, True Float=$System.Timeline001.Time, False Float=1)
    )
    ScaleA=true
    ScaleRGB=true
   ) {module=ScaleColor001}
  ]
  rendererClasses [&NiagaraMeshRendererProperties]
  renderers [{rendererClass=&NiagaraMeshRendererProperties, rendererIndex=0}]
 }
 Fire_spline {
  emitterUpdateScript [
   &EmitterState(
    Inactive Response=::NewEnumerator1
    Life Cycle Mode=::NewEnumerator1
    Loop Behavior=::NewEnumerator1
    Loop Duration Mode=::NewEnumerator1
    UseLoopDelay=false
   )
   &SpawnBurst_Instantaneous(Spawn Count=6, Use Spawn Probability=false, UseLoopCountLimit=false)
  ]
  particleSpawnScript [
   &FN_SetSpawnGroup(Number of Spawn Groups=6)
   &InitializeParticle(
    Color=<rgba 0.65 0.1894 0.5517 0.7391>
    Mesh Scale=&SelectVectorFromArray(
     Direct Array Index=&SpawnGroup()
     Vector Selection Array=DataInterface() {propertyValues {FloatData [<0.3 0.3 0.5>, <0.1 0.1 0.1>, <0.1 0.1 0.1>, <0.3 0.3 0.5>, <0.1 0.1 0.1>, <0.1 0.1 0.1>]}}
    )
    Mesh Scale Mode=::NewEnumerator4
    UsePositionOffset=false
    Write Material Random=true
   )
   &InitialMeshOrientation(
    Mesh Axis To Orient=&SelectVectorFromArray(
     Direct Array Index=&SpawnGroup()
     Vector Selection Array=DataInterface() {propertyValues {FloatData [<0 0 1>, <0 0 1>, <0 0 1>, <1 0 0>, <1 0 0>, <1 0 0>]}}
    )
    Mesh Orientation Mode=::NewEnumerator2
    Orientation Vector=&SelectVectorFromArray(
     Direct Array Index=&SpawnGroup()
     Vector Selection Array=DataInterface() {propertyValues {FloatData [<0 1 0>, <0 1 0>, <0 1 0>, <0 1 0>, <0 1 0>, <0 1 0>]}}
    )
    Rotation=&SelectVectorFromArray(
     Direct Array Index=&SpawnGroup()
     Vector Selection Array=DataInterface() {propertyValues {FloatData [<0 1.1033 -1.3447>, <0 1.1033 -1.3447>, <0 1.1033 -1.3447>, <-0.45 -0.361 -1.3586>, <-0.45 -0.361 -1.3586>, <-0.45 -0.361 -1.3586>]}}
    )
    Use Rotation Vector=true
   )
   &CameraOffset(Camera Offset Amount=1)
  ]
  particleUpdateScript [
   &ParticleState(Kill Particles When Lifetime Has Elapsed=false, Let Infinitely Lived Particles Die When Emitter Deactivates=true)
   &ScaleMeshSize(Scale Factor=&VectorFromFloat(Value=$System.Fade))
   &ScaleMeshSize(Scale Factor=&MakeVector(
     X=1
     Y=1
     Z=&FloatFromCurve(
      FloatCurve=DataInterface() {propertyValues {Curve {defaultValue=1, keys [{interpMode=RCIM_Cubic, tangentMode=RCTM_Break, time=0, value=0.85}, {tangentMode=RCTM_User, time=1, value=1}]}}}
      CurveIndex=$System.AudioReactivity
     )
    )) {module=ScaleMeshSize001}
   &DynamicMaterialParameters(
    Index 0 Param 1=&Lerp_Float(
     A=60
     B=&SelectFloatFromArray(
      Direct Array Index=&SpawnGroup()
      Float Selection Array=DataInterface() {propertyValues {FloatData [80,60,50,60,40,30]}}
     )
     Alpha=$System.AudioReactivity
    )
    Index 0 Param 2=5
    Index 0 Param 3=&Multiply_Float(A=$System.Age, B=&RandomRangeFloat(Minimum=-0.5, Maximum=-0.7))
    Index 0 Param 4=&RandomRangeFloat(Minimum=0.5, Maximum=1)
    Module.Index0Param0Write=true
    Module.Index0Param1Write=true
    Module.Index0Param2Write=true
    Module.Index0Param3Write=true
   )
   /SetVariables(Particles.CustomDynParam_0:Vector4f=&MakeVector4(
     X=15
     Y=3
     Z=&Add_Float(
      A=&Multiply_Float(A=$System.Age, B=-0.7)
      B=&RandomRangeFloat(Minimum=0, Maximum=100)
     )
     W=1
    )) {isSetParameters=true}
   &SocketLocation(
    SamplePosOffset=true
    Sampled Position Offset=&SelectVectorFromArray(
     Direct Array Index=&SpawnGroup()
     Vector Selection Array=DataInterface() {propertyValues {FloatData [<1 -7 -3>, <-1 -3 -2>, <0 -3 -1>, <0 -2 -5>, <0 0 0>, <0 0 0>]}}
    )
    Socket Index=&SelectIntFromArray(Direct Array Index=&SpawnGroup(), Int Selection Array=DataInterface() {propertyValues {IntData [0,1,2,3,4,5]}})
    Socket Reader=$System.Socket Reader
    Socket Sampling Mode=::NewEnumerator3
    VelocitySpeedLimit=false
   )
   &ScaleColor(
    Scale Alpha=&FloatFromCurve(
     FloatCurve=DataInterface() {propertyValues {Curve {
        defaultValue=1
        keys [
         {interpMode=RCIM_Cubic, leaveTangent=0.1774, tangentMode=RCTM_Break, time=0, value=0}
         {arriveTangent=1.9093, interpMode=RCIM_Cubic, tangentMode=RCTM_Break, time=1, value=1}
        ]
       }}}
     CurveIndex=$System.AudioReactivity
     Scale Curve=&MakeCustomFloatFromBool(Bool=$User.bIsFrontEndPreview, True Float=$System.Timeline001.Time, False Float=0.45)
    )
    ScaleA=true
    ScaleRGB=true
   ) {module=ScaleColor002}
  ]
  rendererClasses [&NiagaraMeshRendererProperties]
  renderers [
   {rendererClass=&NiagaraMeshRendererProperties, rendererIndex=0}
   {rendererClass=&NiagaraMeshRendererProperties, rendererIndex=1}
  ]
 }
 Sparkles {
  emitterUpdateScript [
   &EmitterState()
   &SpawnRate(
    Spawn Probability=&MakeCustomFloatFromBool(
     Bool=&SetBoolByFloatComparison(A=$System.AudioReactivity, B=0.1, Comparison Type=::NewEnumerator2)
     True Float=0.25
     False Float=0
    )
    SpawnRate=&FN_FrontEndFloatModifier(Gameplay=15, Front End=10)
    Use Spawn Probability=true
   )
  ]
  particleSpawnScript [
   &FN_SetSpawnGroup()
   &InitializeParticle(
    AdjustAlpha=false
    AdjustHue=true
    AdjustSaturation=false
    AdjustValue=false
    Color=&LerpLinearColors(
     StartColor=<rgba 10 3.5741 9.1143 1>
     EndColor=&ColorFromCurve(
      ColorCurve=DataInterface() {propertyValues {
        AlphaCurve {keys [<at 0 1 RCIM_Cubic>]}
        BlueCurve {keys [
          <at 0.3001 0.8695 RCIM_Cubic>
          {arriveTangent=0.379, interpMode=RCIM_Cubic, leaveTangent=0.379, time=0.3457, value=1}
          <at 0.6445 1>
          <at 0.6875 1.6975>
         ]}
        GreenCurve {keys [
          <at 0.3001 0.3224 RCIM_Cubic>
          {arriveTangent=1.9674, interpMode=RCIM_Cubic, leaveTangent=1.9674, time=0.3457, value=1}
          <at 0.6445 1>
          <at 0.6875 0.4037>
         ]}
        RedCurve {keys [<at 0.3001 1 RCIM_Cubic>,<at 0.3457 1 RCIM_Cubic>,<at 0.6445 1>,<at 0.6875 2>]}
       }}
      CurveIndex=&RandomRangeFloat(Minimum=0, Maximum=1)
     )
     LerpFactor=0
    )
    Color Mode=::NewEnumerator3
    Hue Shift Range:Vector2f=<-0.05 0.05>
    Lifetime Max=0.4
    Lifetime Min=0.3
    Lifetime Mode=::NewEnumerator1
    Position Mode=::NewEnumerator0
    Sprite Rotation Mode=::NewEnumerator2
    Sprite Size Max=<25 25>
    Sprite Size Min=<15 15>
    Sprite Size Mode=::NewEnumerator2
    Write Material Random=true
   )
   &CameraOffset(Camera Offset Amount=5) {module=CameraOffset001}
  ]
  particleUpdateScript [
   &SocketLocation(
    Orientation Sampling=::NewEnumerator1
    SamplePosOffset=true
    Sampled Position Coordinate Space:ENiagaraCoordinateSpace=::Local
    Sampled Position Offset=&MaskVectorBySpawnGroup(
     Masked Vector=&RandomRangeVector(Minimum=<5 0 5>, Maximum=<0 0.5 -10>)
     Passthrough Vector=&MaskVectorBySpawnGroup(
      Masked Vector=&RandomRangeVector(Minimum=<5 0 5>, Maximum=<0 0.5 -10>)
      Passthrough Vector=&MaskVectorBySpawnGroup(
       Masked Vector=&RandomRangeVector(Minimum=<0 0 10>, Maximum=<2 0.2 -8>)
       Passthrough Vector=&RandomRangeVector(Minimum=<0 0 10>, Maximum=<2 0.2 -8>)
       Spawn Group Mask=2
      )
      Spawn Group Mask=1
     )
    )
    Sampled Velocity Coordinate Space:ENiagaraCoordinateSpace=::Simulation
    Socket Index=&SelectIntFromArray(Direct Array Index=&SpawnGroup(), Int Selection Array=DataInterface() {propertyValues {IntData [6,7,8,9]}})
    Socket Reader=$System.Socket Reader
    Socket Sampling Mode=::NewEnumerator3
    VelocitySpeedLimit=false
   )
   &ParticleState()
   &SpriteRotationRate(Rotation Rate=&FloatFromCurve(
     FloatCurve=DataInterface() {propertyValues {Curve {
        defaultValue=0
        keys [
         {arriveTangent=-0.0746, interpMode=RCIM_Cubic, leaveTangent=-0.0746, tangentMode=RCTM_User, time=0, value=0}
         {arriveTangent=6.4174, interpMode=RCIM_Cubic, leaveTangent=-1.1328, tangentMode=RCTM_Break, time=0.35, value=0.9847}
         {arriveTangent=-0.0405, interpMode=RCIM_Cubic, leaveTangent=-0.0405, tangentMode=RCTM_Break, time=1, value=0}
        ]
       }}}
     Scale Curve=360
    ))
   &ScaleSpriteSize(Uniform Curve Scale=1.25)
   &ScaleColor(
    Scale Alpha=&Multiply_Float(
     A=&FloatFromCurve(
      FloatCurve=DataInterface() {propertyValues {Curve {keys [
          {interpMode=RCIM_Cubic, tangentMode=RCTM_User, time=0, value=0}
          {arriveTangent=-2.4058, interpMode=RCIM_Cubic, leaveTangent=-2.4058, tangentMode=RCTM_User, time=0.26, value=0.8421}
          {arriveTangent=-1.1775, interpMode=RCIM_Cubic, leaveTangent=-1.1775, time=0.4613, value=0.0645}
          {arriveTangent=-0.1197, interpMode=RCIM_Cubic, leaveTangent=-0.1197, time=0.5898, value=0.4537}
          <at 1 0 RCIM_Cubic>
         ]}}}
      Scale Curve=&RandomRangeFloat(Minimum=0.1, Maximum=0.5)
     )
     B=&FloatFromCurve(
      FloatCurve=DataInterface() {propertyValues {Curve {keys [{tangentMode=RCTM_User, time=0, value=0}, {tangentMode=RCTM_User, time=1, value=1}]}}}
      Scale Curve=&FloatFromCurve(
       FloatCurve=DataInterface() {propertyValues {Curve {keys [
           {arriveTangentWeight=0.1667, interpMode=RCIM_Cubic, leaveTangentWeight=0.1667, tangentMode=RCTM_User, time=1.4, value=0}
           {tangentMode=RCTM_User, time=2, value=1}
          ]}}}
       CurveIndex=$Emitter.Age
      )
     )
    )
    Scale RGB=&VectorFromFloat(Value=&Multiply_Float(
      A=&RandomRangeFloat(Minimum=0.7, Maximum=0.9, Evaluation Type=::NewEnumerator1)
      B=&SelectFloatFromArray(Direct Array Index=$Engine.QualityLevel, Float Selection Array=DataInterface() {propertyValues {FloatData [10,5,1,1,1]}})
     ))
    ScaleA=true
    ScaleRGB=true
   ) {module=ScaleColor001}
   &ScaleColor(Scale Alpha=1.25, ScaleA=true, ScaleRGB=true)
   /SetVariables(
    Particles.CenterFlareColor:LinearColor=&MultiplyLinearColor(
     Linear Color=$Particles.Color
     Scale Factor=&MakeVector4_FromLinearColor(LinearColor=&LerpLinearColors(StartColor=<rgba 10 3.937 9.5847 1>, EndColor=<rgba 10 7.7389 9.6274 1>, LerpFactor=0))
    )
    Particles.CenterFlareSize:Vector2f=&Multiply_Vector2(A=$Particles.SpriteSize, B=&Vector2DFromFloat(Value=0.65))
    Particles.CoronaFlareColor:LinearColor=&MultiplyLinearColor(
     Linear Color=$Particles.Color
     Scale Factor=&MakeVector4_FromLinearColor(LinearColor=&LerpLinearColors(StartColor=<rgba 0.15 0.15 0.15 1>, EndColor=<rgba 0.0729 0.0729 0.0729 1>, LerpFactor=0))
    )
    Particles.CoronaFlareSize:Vector2f=&Multiply_Vector2(A=$Particles.Initial.SpriteSize, B=&Vector2DFromFloat(Value=1))
    Particles.FlickerFlareColor:LinearColor=&MultiplyLinearColor(
     Linear Color=$Particles.Color
     Scale Factor=&MakeVector4_FromLinearColor(LinearColor=&LerpLinearColors(StartColor=<rgba 0.9441 0.8079 1 1>, EndColor=<rgba 0.0337 0.0417 0.0337 50>, LerpFactor=0))
    )
   ) {isSetParameters=true}
  ]
  rendererClasses [&NiagaraSpriteRendererProperties]
  renderers [
   {rendererClass=&NiagaraSpriteRendererProperties}
   {rendererClass=&NiagaraSpriteRendererProperties, rendererIndex=1}
   {rendererClass=&NiagaraSpriteRendererProperties, rendererIndex=2}
  ]
 }
 Sparks {
  emitterUpdateScript [
   &EmitterState()
   &SpawnRate(
    SpawnRate=&Lerp_Float(
     A=0
     B=&FloatFromCurve(
      FloatCurve=DataInterface() {propertyValues {Curve {defaultValue=1, keys [{interpMode=RCIM_Cubic, tangentMode=RCTM_Break, time=0, value=22}, {tangentMode=RCTM_User, time=1, value=35}]}}}
      CurveIndex=$System.AudioReactivity
     )
     Alpha=$System.AudioReactivity
    )
    Use Spawn Probability=false
   )
  ]
  particleSpawnScript [
   &FN_SetSpawnGroup(Number of Spawn Groups=6)
   &InitializeParticle(
    AdjustAlpha=false
    AdjustHue=true
    AdjustSaturation=false
    AdjustValue=true
    Color=<rgba 5 0.8354 3.7092 1>
    Color Mode=::NewEnumerator3
    Hue Shift Range:Vector2f=<-0.05 0.05>
    Lifetime Max=0.75
    Lifetime Min=0.4
    Lifetime Mode=::NewEnumerator1
    Sprite Size=&Multiply_Vector2D_ByFloat(Float=&RandomRangeFloat(Minimum=0.35, Maximum=1), Vector2D=<3 8>)
    Sprite Size Mode=::NewEnumerator4
    UsePositionOffset=false
    Write Material Random=true
   )
   &SocketLocation(
    SamplePosOffset=true
    Socket Index=&SelectIntFromArray(Direct Array Index=&SpawnGroup(), Int Selection Array=DataInterface() {propertyValues {IntData [0,1,2,3,4,5]}})
    Socket Reader=$System.Socket Reader
    Socket Sampling Mode=::NewEnumerator3
    VelocitySpeedLimit=false
   )
   &AddVelocity(
    Cone Angle=25
    Cone Axis=&MakeCustomVectorFromBool(
     Bool=$User.bIsFrontEndPreview
     True Vector=&SelectVectorFromArray(
      Direct Array Index=&SpawnGroup()
      Vector Selection Array=DataInterface() {propertyValues {FloatData [<1 -1 1.25>, <1 -1 1.25>, <1 -1 1.25>, <0.25 -1 1.25>, <0.25 -1 1.25>, <0.25 -1 1.25>]}}
     )
     False Vector=&SelectVectorFromArray(
      Direct Array Index=&SpawnGroup()
      Vector Selection Array=DataInterface() {propertyValues {FloatData [<1 -1 1.25>, <1 -1 1.25>, <1 -1 1.25>, <-0.125 -1 1>, <-0.125 -1 1>, <-0.125 -1 1>]}}
     )
    )
    Module.UseConeFalloff=false
    Module.UseVelDistribution=false
    Override Seed=false
    Velocity Mode=::NewEnumerator2
    Velocity Speed=&Multiply_Float(
     A=&FloatFromCurve(
      FloatCurve=DataInterface() {propertyValues {Curve {
         defaultValue=1
         keys [
          {interpMode=RCIM_Cubic, tangentMode=RCTM_Break, time=0, value=25}
          {arriveTangent=-12.5853, interpMode=RCIM_Cubic, tangentMode=RCTM_Break, time=1, value=100}
         ]
        }}}
      CurveIndex=$System.AudioReactivity
     )
     B=&RandomRangeFloat(Minimum=0.35, Maximum=1)
    )
   )
  ]
  particleUpdateScript [
   &ParticleState()
   /Niagara/Modules/Update/Forces/CurlNoiseForce(Noise Frequency=100, Noise Strength=&Multiply_Float(A=1000, B=0.25), Pan Noise=true)
   &SolveForcesAndVelocity(Clamp Velocity=false, Limit Acceleration=false)
   &ScaleColor(
    Scale Alpha=&FloatFromCurve(
     FloatCurve=DataInterface() {propertyValues {Curve {
        defaultValue=1
        keys [
         {interpMode=RCIM_Cubic, leaveTangent=1.4778, tangentMode=RCTM_Break, time=0, value=0}
         {interpMode=RCIM_Cubic, tangentMode=RCTM_Break, time=0.125, value=1.0147}
         {arriveTangent=-0.0405, interpMode=RCIM_Cubic, leaveTangent=-0.0405, tangentMode=RCTM_User, time=0.7009, value=1}
         {arriveTangent=-8.0089, interpMode=RCIM_Cubic, leaveTangent=-8.0089, tangentMode=RCTM_User, time=1, value=0}
        ]
       }}}
     Scale Curve=&MakeCustomFloatFromBool(Bool=$User.bIsFrontEndPreview, True Float=$System.Timeline001.Time, False Float=1)
    )
    Scale RGB=&MakeCustomVectorFromBool(
     Bool=$User.bIsFrontEndPreview
     True Vector=&Lerp_Vector(
      A=&VectorFromFloat(Value=&Lerp_Float(
        A=1
        B=10
        Alpha=&FloatFromCurve(
         FloatCurve=DataInterface() {propertyValues {Curve {defaultValue=0, keys [<at 0 0 RCIM_Constant>, {tangentMode=RCTM_User, time=0.9, value=1}]}}}
         CurveIndex=&RandomRangeFloat(Minimum=0, Maximum=1)
        )
       ))
      B=<1 0 0>
      Alpha=$Particles.NormalizedAge
     )
     False Vector=&Lerp_Vector(
      A=&VectorFromFloat(Value=&Lerp_Float(
        A=1
        B=150
        Alpha=&FloatFromCurve(
         FloatCurve=DataInterface() {propertyValues {Curve {defaultValue=0, keys [<at 0 0 RCIM_Constant>, {tangentMode=RCTM_User, time=0.9, value=1}]}}}
         CurveIndex=&RandomRangeFloat(Minimum=0, Maximum=1)
        )
       ))
      B=<1 0 0>
      Alpha=$Particles.NormalizedAge
     )
    )
    ScaleA=true
    ScaleRGB=true
   ) {module=ScaleColor001}
  ]
  rendererClasses [&NiagaraSpriteRendererProperties]
  renderers [{rendererClass=&NiagaraSpriteRendererProperties}]
 }
 Sparks_movement {
  emitterUpdateScript [
   &EmitterState()
   &SpawnPerUnit(
    Module.Use Max Movement Threshold=true
    Module.Use Movement Tolerance=true
    Spawn Probability:NiagaraFloat=&MakeCustomFloatFromBool(
     Bool=$User.bIsFrontEndPreview
     True Float=0
     False Float=&MakeCustomFloatFromBool(
      Bool=&SetBoolByFloatComparison(A=$System.AudioReactivity, B=0.1, Comparison Type=::NewEnumerator2)
      True Float=1
      False Float=0
     )
    )
    Spawn Spacing=75
    Use Spawn Probability=true
   )
  ]
  particleSpawnScript [
   &FN_SetSpawnGroup(Number of Spawn Groups=6)
   &InitializeParticle(
    AdjustAlpha=false
    AdjustHue=true
    AdjustSaturation=false
    AdjustValue=true
    Color=<rgba 5 0.8354 3.7092 1>
    Color Mode=::NewEnumerator3
    Hue Shift Range:Vector2f=<-0.05 0.05>
    Lifetime Max=0.75
    Lifetime Min=0.4
    Lifetime Mode=::NewEnumerator1
    Sprite Size=&Multiply_Vector2D_ByFloat(Float=&RandomRangeFloat(Minimum=0.75, Maximum=1), Vector2D=<4 8>)
    Sprite Size Mode=::NewEnumerator4
    UsePositionOffset=false
    Write Material Random=true
   )
   &SocketLocation(
    SamplePosOffset=true
    Socket Index=&SelectIntFromArray(Direct Array Index=&SpawnGroup(), Int Selection Array=DataInterface() {propertyValues {IntData [0,1,2,3,4,5]}})
    Socket Reader=$System.Socket Reader
    Socket Sampling Mode=::NewEnumerator3
    VelocitySpeedLimit=false
   )
   &AddVelocity(
    Cone Angle=25
    Cone Axis=&SelectVectorFromArray(
     Direct Array Index=&SpawnGroup()
     Vector Selection Array=DataInterface() {propertyValues {FloatData [<1 -1 1.25>, <1 -1 1.25>, <1 -1 1.25>, <-0.125 -1 1>, <-0.125 -1 1>, <-0.125 -1 1>]}}
    )
    Module.UseConeFalloff=false
    Module.UseVelDistribution=false
    Override Seed=false
    Velocity Mode=::NewEnumerator2
    Velocity Speed=&Multiply_Float(
     A=&FloatFromCurve(
      FloatCurve=DataInterface() {propertyValues {Curve {
         defaultValue=1
         keys [
          {interpMode=RCIM_Cubic, tangentMode=RCTM_Break, time=0, value=25}
          {arriveTangent=-12.5853, interpMode=RCIM_Cubic, tangentMode=RCTM_Break, time=1, value=100}
         ]
        }}}
      CurveIndex=$System.AudioReactivity
     )
     B=&RandomRangeFloat(Minimum=0.35, Maximum=1)
    )
   )
  ]
  particleUpdateScript [
   &ParticleState()
   &CurlNoiseForce(
    Noise Frequency=100
    Noise Strength=&Multiply_Float(A=1000, B=&RandomRangeFloat(Minimum=0.2, Maximum=0.5))
    Pan Noise=true
   )
   &Drag(Use Linear Drag=true, Use Rotational Drag=true)
   &InheritVelocity(Inherited Velocity Speed Limit=-15, Speed Limit=true, UseSpeedThreshold=false)
   &SolveForcesAndVelocity(Clamp Velocity=false, Limit Acceleration=false)
   &ScaleColor(
    Scale Alpha=&FloatFromCurve(FloatCurve=DataInterface() {propertyValues {Curve {
        defaultValue=1
        keys [
         {interpMode=RCIM_Cubic, leaveTangent=1.4778, tangentMode=RCTM_Break, time=0, value=0}
         {interpMode=RCIM_Cubic, tangentMode=RCTM_Break, time=0.125, value=1.0147}
         {arriveTangent=-0.0405, interpMode=RCIM_Cubic, leaveTangent=-0.0405, tangentMode=RCTM_User, time=0.7009, value=1}
         {arriveTangent=-8.0089, interpMode=RCIM_Cubic, leaveTangent=-8.0089, tangentMode=RCTM_User, time=1, value=0}
        ]
       }}})
    Scale RGB=&Lerp_Vector(
     A=&VectorFromFloat(Value=&Lerp_Float(
       A=1
       B=150
       Alpha=&FloatFromCurve(
        FloatCurve=DataInterface() {propertyValues {Curve {defaultValue=0, keys [<at 0 0 RCIM_Constant>, {tangentMode=RCTM_User, time=0.9, value=1}]}}}
        CurveIndex=&RandomRangeFloat(Minimum=0, Maximum=1)
       )
      ))
     B=<1 0 0>
     Alpha=$Particles.NormalizedAge
    )
    ScaleA=true
    ScaleRGB=true
   ) {module=ScaleColor001}
  ]
  rendererClasses [&NiagaraSpriteRendererProperties]
  renderers [{rendererClass=&NiagaraSpriteRendererProperties}]
 }
}
summary {
 systemName=NS_Character_Body_MistTycoonLife_Boots
 emitters [
  {emitterName=Fire_Mesh, rendererClasses [&NiagaraMeshRendererProperties]}
  {emitterName=Fire_spline, rendererClasses [&NiagaraMeshRendererProperties]}
  {emitterName=Sparks, rendererClasses [&NiagaraSpriteRendererProperties]}
  {emitterName=Sparks_movement, rendererClasses [&NiagaraSpriteRendererProperties]}
  {emitterName=Sparkles, rendererClasses [&NiagaraSpriteRendererProperties]}
 ]
 userVariables [
  User.bisFootHidden {type=NiagaraBool}
  User.bIsFrontEnd {type=NiagaraBool, defaultValue=true}
  User.bIsFrontEndPreview {type=NiagaraBool}
  User.EnableFrontendFakeAudioLoop {type=NiagaraFloat, defaultValue=0, description=""}
 ]
}
