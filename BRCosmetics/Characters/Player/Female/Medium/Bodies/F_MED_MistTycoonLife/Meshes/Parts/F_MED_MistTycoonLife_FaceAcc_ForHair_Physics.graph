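# Physics asset for the FaceAcc_ForHair mesh part of the female medium MistTycoonLife character body, setting up a side-hair chain.
# Four hair_side capsules link in sequence from MSTR_hair_side, with a wider swing limit at the root joint and tight twist limits throughout.
# Unconstrained capsules on the head, neck, upper spine, clavicles and upper arms give body shapes around the hair chain.
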PhysicsAsset /BRCosmetics/Characters/Player/Female/Medium/Bodies/F_MED_MistTycoonLife/Meshes/Parts/F_MED_MistTycoonLife_FaceAcc_ForHair_Physics
bodies {
 MSTR_hair_side [{shapeType=Capsule, length=2, radius=1, shapeName=MSTR_hair_side_capsule}]
 clavicle_l [{shapeType=Capsule, length=5.2033, radius=4.3101, shapeName=clavicle_l_capsule, center=<5.4639 0 0>, rotation=<rot 90 0 0>}]
 clavicle_r [{
   shapeType=Capsule
   length=5.9899
   radius=3.5235
   shapeName=clavicle_r_capsule
   center=<-5.4375 0.5919 -0.5936>
   rotation=<rot 83.7505 -0.001 -0.0011>
  }]
 hair_side_1 [{shapeType=Capsule, length=2.3658, radius=2.6342, shapeName=hair_side_1_capsule, center=<3.8782 0 0>, rotation=<rot 90 0 0>}]
 hair_side_2 [{shapeType=Capsule, length=1.48, radius=3, shapeName=hair_side_2_capsule, center=<3.8782 -1.4256 -0.778>, rotation=<rot 90 0 0>}]
 hair_side_3 [{shapeType=Capsule, length=0.7338, radius=3.2662, shapeName=hair_side_3_capsule, center=<2.8768 -2.2945 0>, rotation=<rot 90 0 0>}]
 hair_side_4 [{shapeType=Capsule, length=0.935, radius=4.065, shapeName=hair_side_4_capsule, center=<3.8782 -2.127 0>, rotation=<rot 90 0 0>}]
 head [{shapeType=Capsule, length=4.4658, radius=10.2189, shapeName=head_capsule, center=<-0.556 3.9928 0>}]
 neck_01 [{shapeType=Capsule, length=1.913, radius=5.0263, shapeName=neck_01_capsule, rotation=<rot 90 0 0>}]
 spine_03 [{shapeType=Capsule, length=9.5, radius=7, shapeName=spine_03_capsule}]
 spine_04 [
  {shapeType=Capsule, length=7.7257, radius=7.8172, shapeName=spine_04_capsule, center=<2.0616 8.2592 0>}
  {shapeType=Capsule, length=22.5, radius=7.5, shapeName=spine_04_capsule}
 ]
 spine_05 [{shapeType=Capsule, length=12.954, radius=7.7705, shapeName=spine_05_capsule, center=<2.7235 -0.349 0>}]
 upperarm_l [{shapeType=Capsule, length=12.607, radius=4.393, shapeName=upperarm_l_capsule, center=<5 -0.4274 0>, rotation=<rot 90 0 0>}]
 upperarm_r [{shapeType=Capsule, length=18.2645, radius=4.0115, shapeName=upperarm_r_capsule, center=<-9.1197 0.8942 0.6198>, rotation=<rot 90 0 0>}]
}
constraints [
 {bone1Name=hair_side_1, bone2Name=MSTR_hair_side, swing1LimitDegrees=50, swing2LimitDegrees=50, twistLimitDegrees=2}
 {bone1Name=hair_side_2, bone2Name=hair_side_1, swing1LimitDegrees=30, swing2LimitDegrees=30, twistLimitDegrees=2}
 {bone1Name=hair_side_3, bone2Name=hair_side_2, swing1LimitDegrees=30, swing2LimitDegrees=30, twistLimitDegrees=2}
 {bone1Name=hair_side_4, bone2Name=hair_side_3, swing1LimitDegrees=30, swing2LimitDegrees=30, twistLimitDegrees=2}
]
