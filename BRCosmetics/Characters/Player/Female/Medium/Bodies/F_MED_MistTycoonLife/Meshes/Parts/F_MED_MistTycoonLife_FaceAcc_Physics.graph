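# Physics asset for the MistTycoonLife female FaceAcc mesh, setting up physics for its side and back hair.
# Four capsule chains (side, back left, middle, right) hang from MSTR_hair bones with limited swing and twist.
# Free-angular cross constraints tie left and right back segments to matching middle segments, linking the back hair laterally.
# Capsules on head, neck, clavicles, upper arms and spine provide body shapes for hair collision.
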PhysicsAsset /BRCosmetics/Characters/Player/Female/Medium/Bodies/F_MED_MistTycoonLife/Meshes/Parts/F_MED_MistTycoonLife_FaceAcc_Physics
bodies {
 MSTR_hair_back_l [{shapeType=Capsule, length=1.6562, radius=0.9324, shapeName=MSTR_hair_back_l_capsule}]
 MSTR_hair_back_r [{shapeType=Capsule, length=1.6562, radius=0.9324, shapeName=MSTR_hair_back_r_capsule}]
 MSTR_hair_mid_back [{shapeType=Capsule, length=1.6562, radius=0.9324, shapeName=MSTR_hair_mid_back_capsule}]
 MSTR_hair_side [{shapeType=Capsule, length=2, radius=1, shapeName=MSTR_hair_side_capsule}]
 clavicle_l [{shapeType=Capsule, length=5.2033, radius=4.3101, shapeName=clavicle_l_capsule, center=<5.4639 0 0>, rotation=<rot 90 0 0>}]
 clavicle_r [{
   shapeType=Capsule
   length=5.9899
   radius=3.5235
   shapeName=clavicle_r_capsule
   center=<-5.4375 0.5919 -0.5936>
   rotation=<rot 83.7505 -0.001 -0.0011>
  }]
 hair_back_l_1 [{shapeType=Capsule, length=3.0754, radius=2.9246, shapeName=hair_back_l_1_capsule, center=<3.303 -0.4669 0>}]
 hair_back_l_2 [{shapeType=Capsule, length=3, radius=3, shapeName=hair_back_l_2_capsule, center=<3.303 -1.5 0>}]
 hair_back_l_3 [{shapeType=Capsule, length=3, radius=3, shapeName=hair_back_l_3_capsule, center=<3.303 -1.5 0>}]
 hair_back_l_4 [{shapeType=Capsule, length=3, radius=3, shapeName=hair_back_l_4_capsule, center=<4.5542 -1 0>}]
 hair_back_l_5 [{shapeType=Capsule, length=3, radius=3, shapeName=hair_back_l_5_capsule, center=<3.303 -1 0>}]
 hair_back_mid_1 [{shapeType=Capsule, length=3.0754, radius=2.9246, shapeName=hair_back_mid_1_capsule, center=<3.303 -0.4669 0>}]
 hair_back_mid_2 [{shapeType=Capsule, length=3, radius=3, shapeName=hair_back_mid_2_capsule, center=<3.303 -1.5 0>}]
 hair_back_mid_3 [{shapeType=Capsule, length=3, radius=3, shapeName=hair_back_mid_3_capsule, center=<3.303 -1.5 0>}]
 hair_back_mid_4 [{shapeType=Capsule, length=3, radius=3, shapeName=hair_back_mid_4_capsule, center=<4.5542 -1 0>}]
 hair_back_mid_5 [{shapeType=Capsule, length=3, radius=3, shapeName=hair_back_mid_5_capsule, center=<3.303 -1 0>}]
 hair_back_mid_6 [{shapeType=Capsule, length=3, radius=3, shapeName=hair_back_mid_6_capsule, center=<3.303 0 0>}]
 hair_back_r_1 [{shapeType=Capsule, length=3.0754, radius=2.9246, shapeName=hair_back_r_1_capsule, center=<3.303 -0.4669 0>}]
 hair_back_r_2 [{shapeType=Capsule, length=3, radius=3, shapeName=hair_back_r_2_capsule, center=<3.303 -1.5 0>}]
 hair_back_r_3 [{shapeType=Capsule, length=3, radius=3, shapeName=hair_back_r_3_capsule, center=<3.303 -1.5 0>}]
 hair_back_r_4 [{shapeType=Capsule, length=3, radius=3, shapeName=hair_back_r_4_capsule, center=<4.5542 -1 0>}]
 hair_back_r_5 [{shapeType=Capsule, length=3, radius=3, shapeName=hair_back_r_5_capsule, center=<3.303 -1 0>}]
 hair_side_1 [{shapeType=Capsule, length=2.3658, radius=2.6342, shapeName=hair_side_1_capsule, center=<3.8782 0 0>, rotation=<rot 90 0 0>}]
 hair_side_2 [{shapeType=Capsule, length=1.48, radius=3, shapeName=hair_side_2_capsule, center=<3.8782 -1.4256 -0.778>, rotation=<rot 90 0 0>}]
 hair_side_3 [{shapeType=Capsule, length=0.7338, radius=3.2662, shapeName=hair_side_3_capsule, center=<2.8768 -2.2945 0>, rotation=<rot 90 0 0>}]
 hair_side_4 [{shapeType=Capsule, length=0.935, radius=4.065, shapeName=hair_side_4_capsule, center=<3.8782 -2.127 0>, rotation=<rot 90 0 0>}]
 head [{shapeType=Capsule, length=4.4658, radius=10.2189, shapeName=head_capsule, center=<-0.556 3.9928 0>}]
 neck_01 [{shapeType=Capsule, length=1.913, radius=5.0263, shapeName=neck_01_capsule, rotation=<rot 90 0 0>}]
 spine_02 [{shapeType=Capsule, length=9.5, radius=7, shapeName=spine_02_capsule}]
 spine_03 [{shapeType=Capsule, length=9.5, radius=7, shapeName=spine_03_capsule}]
 spine_04 [
  {shapeType=Capsule, length=10.3386, radius=7.3432, shapeName=spine_04_capsule, center=<2.0269 5.9511 0>}
  {shapeType=Capsule, length=22.5, radius=7.5, shapeName=spine_04_capsule}
 ]
 spine_05 [
  {shapeType=Capsule, length=13.5218, radius=7.2028, shapeName=spine_05_capsule, center=<2.7235 -1.0888 0>}
  {shapeType=Capsule, length=15.4958, radius=5.1873, shapeName=spine_05_capsule, center=<1.5795 3.4999 0>}
 ]
 upperarm_l [{shapeType=Capsule, length=12.607, radius=4.393, shapeName=upperarm_l_capsule, center=<5 -0.4274 0>, rotation=<rot 90 0 0>}]
 upperarm_r [{shapeType=Capsule, length=18.2645, radius=4.0115, shapeName=upperarm_r_capsule, center=<-9.1197 0.8942 0.6198>, rotation=<rot 90 0 0>}]
}
constraints [
 {bone1Name=hair_side_1, bone2Name=MSTR_hair_side, swing1LimitDegrees=50, swing2LimitDegrees=50, twistLimitDegrees=2}
 {bone1Name=hair_side_2, bone2Name=hair_side_1, swing1LimitDegrees=30, swing2LimitDegrees=30, twistLimitDegrees=2}
 {bone1Name=hair_side_3, bone2Name=hair_side_2, swing1LimitDegrees=30, swing2LimitDegrees=30, twistLimitDegrees=2}
 {bone1Name=hair_side_4, bone2Name=hair_side_3, swing1LimitDegrees=30, swing2LimitDegrees=30, twistLimitDegrees=2}
 {bone1Name=hair_back_l_1, bone2Name=MSTR_hair_back_l, swing1LimitDegrees=35, swing2LimitDegrees=30, twistLimitDegrees=5}
 {bone1Name=hair_back_l_2, bone2Name=hair_back_l_1, swing1LimitDegrees=25, swing2LimitDegrees=30, twistLimitDegrees=5}
 {bone1Name=hair_back_l_3, bone2Name=hair_back_l_2, swing1LimitDegrees=30, swing2LimitDegrees=30, twistLimitDegrees=5}
 {bone1Name=hair_back_l_4, bone2Name=hair_back_l_3, swing1LimitDegrees=30, swing2LimitDegrees=30, twistLimitDegrees=5}
 {bone1Name=hair_back_l_5, bone2Name=hair_back_l_4, swing1LimitDegrees=30, swing2LimitDegrees=30, twistLimitDegrees=5}
 {bone1Name=hair_back_r_1, bone2Name=MSTR_hair_back_r, swing1LimitDegrees=35, swing2LimitDegrees=30, twistLimitDegrees=5}
 {bone1Name=hair_back_r_2, bone2Name=hair_back_r_1, swing1LimitDegrees=25, swing2LimitDegrees=30, twistLimitDegrees=5}
 {bone1Name=hair_back_r_3, bone2Name=hair_back_r_2, swing1LimitDegrees=30, swing2LimitDegrees=30, twistLimitDegrees=5}
 {bone1Name=hair_back_r_4, bone2Name=hair_back_r_3, swing1LimitDegrees=30, swing2LimitDegrees=30, twistLimitDegrees=5}
 {bone1Name=hair_back_r_5, bone2Name=hair_back_r_4, swing1LimitDegrees=30, swing2LimitDegrees=30, twistLimitDegrees=5}
 {bone1Name=hair_back_mid_1, bone2Name=MSTR_hair_mid_back, swing1LimitDegrees=35, swing2LimitDegrees=25, twistLimitDegrees=3}
 {bone1Name=hair_back_mid_2, bone2Name=hair_back_mid_1, swing1LimitDegrees=25, swing2LimitDegrees=20, twistLimitDegrees=3}
 {bone1Name=hair_back_mid_3, bone2Name=hair_back_mid_2, swing1LimitDegrees=25, swing2LimitDegrees=10, twistLimitDegrees=3}
 {bone1Name=hair_back_mid_4, bone2Name=hair_back_mid_3, swing1LimitDegrees=25, swing2LimitDegrees=10, twistLimitDegrees=3}
 {bone1Name=hair_back_mid_5, bone2Name=hair_back_mid_4, swing1LimitDegrees=25, swing2LimitDegrees=10, twistLimitDegrees=3}
 {bone1Name=hair_back_mid_6, bone2Name=hair_back_mid_5, swing1LimitDegrees=25, swing2LimitDegrees=10, twistLimitDegrees=3}
 {
  bone1Name=hair_back_r_1
  bone2Name=hair_back_mid_1
  swing1LimitDegrees=45
  swing1Motion=Free
  swing2LimitDegrees=45
  swing2Motion=Free
  twistLimitDegrees=45
  twistMotion=Free
 }
 {
  bone1Name=hair_back_r_2
  bone2Name=hair_back_mid_2
  swing1LimitDegrees=45
  swing1Motion=Free
  swing2LimitDegrees=45
  swing2Motion=Free
  twistLimitDegrees=45
  twistMotion=Free
 }
 {
  bone1Name=hair_back_r_3
  bone2Name=hair_back_mid_3
  swing1LimitDegrees=45
  swing1Motion=Free
  swing2LimitDegrees=45
  swing2Motion=Free
  twistLimitDegrees=45
  twistMotion=Free
 }
 {
  bone1Name=hair_back_r_4
  bone2Name=hair_back_mid_4
  swing1LimitDegrees=45
  swing1Motion=Free
  swing2LimitDegrees=45
  swing2Motion=Free
  twistLimitDegrees=45
  twistMotion=Free
 }
 {
  bone1Name=hair_back_r_5
  bone2Name=hair_back_mid_5
  swing1LimitDegrees=45
  swing1Motion=Free
  swing2LimitDegrees=45
  swing2Motion=Free
  twistLimitDegrees=45
  twistMotion=Free
 }
 {
  bone1Name=hair_back_l_5
  bone2Name=hair_back_mid_5
  swing1LimitDegrees=45
  swing1Motion=Free
  swing2LimitDegrees=45
  swing2Motion=Free
  twistLimitDegrees=45
  twistMotion=Free
 }
 {
  bone1Name=hair_back_l_4
  bone2Name=hair_back_mid_4
  swing1LimitDegrees=45
  swing1Motion=Free
  swing2LimitDegrees=45
  swing2Motion=Free
  twistLimitDegrees=45
  twistMotion=Free
 }
 {
  bone1Name=hair_back_l_3
  bone2Name=hair_back_mid_3
  swing1LimitDegrees=45
  swing1Motion=Free
  swing2LimitDegrees=45
  swing2Motion=Free
  twistLimitDegrees=45
  twistMotion=Free
 }
 {
  bone1Name=hair_back_l_2
  bone2Name=hair_back_mid_2
  swing1LimitDegrees=45
  swing1Motion=Free
  swing2LimitDegrees=45
  swing2Motion=Free
  twistLimitDegrees=45
  twistMotion=Free
 }
 {
  bone1Name=hair_back_l_1
  bone2Name=hair_back_mid_1
  swing1LimitDegrees=45
  swing1Motion=Free
  swing2LimitDegrees=45
  swing2Motion=Free
  twistLimitDegrees=45
  twistMotion=Free
 }
]
