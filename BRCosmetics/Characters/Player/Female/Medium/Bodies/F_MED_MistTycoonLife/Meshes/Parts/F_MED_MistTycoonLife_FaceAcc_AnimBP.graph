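# Animation Blueprint for the MistTycoonLife female face-accessory part, driving hair motion.
# It copies the parent pose, adjusts master hair bones by head-up rotation, applies upper-arm transforms from the body mesh, then runs a Control Rig.
# Back-bling presence selects between two RigidBody simulations with different physics assets, using parameter-driven gravity and sim settings.
# Update logic also computes head left/right alphas, but their bone-adjustment nodes are disconnected from the output.
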AnimBlueprint /BRCosmetics/Characters/Player/Female/Medium/Bodies/F_MED_MistTycoonLife/Meshes/Parts/F_MED_MistTycoonLife_FaceAcc_AnimBP
target_skeleton=&/F_MED_MistTycoonLife_FaceAcc_Skeleton
graphs [
 AnimGraph {nodes [
   "Constraint - hair_shoulder_r_deform (P)"(Alpha=Get Head Right Hair() -> (Head Right Hair:double), ConstraintWeights_0=1) {boneToModify {boneName=hair_shoulder_r_deform}, constraintSetup [{targetBone {boneName=hair_back_r_1}}], constraintWeights [1]}
   "Constraint - hair_shoulder_l_deform (P)"(Alpha=Get Head Left hair() -> (Head Left hair:double), ConstraintWeights_0=1) {boneToModify {boneName=hair_shoulder_l_deform}, constraintSetup [{targetBone {boneName=hair_back_l_3}}], constraintWeights [1]}
   n1 = Get Head Right() -> (Head Right:double)
   n2 = "Transform (Modify) Bone - Bone: MSTR_hair_back_l"(Alpha=@n1, Translation="0,0,0", Rotation="-50,0,0", Scale="1,1,1") {rotationMode=BMM_Additive, rotationSpace=BCS_BoneSpace, boneToModify {boneName=MSTR_hair_back_l}, rotation {pitch=-50}}
   n3 = "Transform (Modify) Bone - Bone: MSTR_hair_mid_back"(ComponentPose=@n2, Alpha=@n1, Translation="0,0,0", Rotation="-50,0,0", Scale="1,1,1") {rotationMode=BMM_Additive, rotationSpace=BCS_BoneSpace, boneToModify {boneName=MSTR_hair_mid_back}, rotation {pitch=-50}}
   "Transform (Modify) Bone - Bone: MSTR_hair_back_r"(ComponentPose=@n3, Alpha=@n1, Translation="0,0,0", Rotation="-50,0,0", Scale="1,1,1") {rotationMode=BMM_Additive, rotationSpace=BCS_BoneSpace, boneToModify {boneName=MSTR_hair_back_r}, rotation {pitch=-50}}
   n4 = RigidBody!(
    ComponentPose=Local To Component(LocalPose=Use cached pose 'Pose'())
    bAlphaBoolEnabled:bool
    OverrideWorldGravity:Vector=Get Gravity Override()
    ExternalForce
    SimSpaceSettings:Sim Space Settings Structure=Get!() -> (
     RBAN Settings Out_ComponentLinearAccScale:Vector=~
     RBAN Settings Out_ComponentLinearVelScale:Vector=~
     RBAN Settings Out_SimSpaceSettings:Sim Space Settings Structure
    ).RBAN Settings Out_SimSpaceSettings
   ) -> (Pose) {
    alphaInputType=Bool
    bOverrideWorldGravity=true
    overridePhysicsAsset=&/F_MED_MistTycoonLife_FaceAcc_ForHair_Physics
    simulationSpace=BaseBoneSpace
    baseBoneRef {boneName=spine_04}
    componentAppliedLinearAccClamp=<0 0 0>
    simSpaceSettings {maxAngularAcceleration=6, maxAngularVelocity=8, maxLinearAcceleration=100, maxLinearVelocity=200, worldAlpha=1}
   }
   n5 = RigidBody(
    ComponentPose=Local To Component(LocalPose=Use cached pose 'Pose'())
    OverrideWorldGravity=Get Gravity Override()
    SimSpaceSettings=Get() -> (RBAN Settings Out_ComponentLinearAccScale=~, RBAN Settings Out_ComponentLinearVelScale=~).RBAN Settings Out_SimSpaceSettings
   ) {
    alphaInputType=Bool
    bOverrideWorldGravity=true
    overridePhysicsAsset=&/F_MED_MistTycoonLife_FaceAcc_Physics
    simulationSpace=BaseBoneSpace
    baseBoneRef {boneName=spine_04}
    componentAppliedLinearAccClamp=<0 0 0>
    simSpaceSettings {maxAngularAcceleration=6, maxAngularVelocity=8, maxLinearAcceleration=100, maxLinearVelocity=200, worldAlpha=1}
   }
   Output Pose(Result=Blend Poses by bool(
     BlendPose_0=Component To Local(ComponentPose=@n4)
     BlendPose_1=Component To Local(ComponentPose=@n5)
     bActiveValue=Get has backpack() -> (has backpack:bool)
    )) {layerGroup=None}
   n6 = Get!() -> (upperarm_l_t_Location:Vector, upperarm_l_t_Rotation:Rotator, upperarm_l_t_Scale:Vector)
   n7 = Get Head Up() -> (Head Up:double)
   n8 = "Constraint - MSTR_hair_mid_back (RR)"!(
    ComponentPose=Local To Component(LocalPose=Copy Pose From Mesh() {bUseAttachedParent=true})
    Alpha=1
    ConstraintWeights_0=0
    ConstraintWeights_1:float=1
   ) -> (Pose) {
    alphaBoolBlend {blendOption=Linear}
    boneToModify {boneName=MSTR_hair_mid_back}
    constraintSetup [{transformType=Rotation, targetBone {boneName=spine_05}}, {transformType=Rotation, targetBone {boneName=MSTR_hair_mid_back}}]
    constraintWeights [1,1]
   }
   n9 = "Constraint - MSTR_hair_back_r (RR)"!(ComponentPose=@n8, Alpha=1, ConstraintWeights_0=0, ConstraintWeights_1:float=1) -> (Pose) {
    alphaBoolBlend {blendOption=Linear}
    boneToModify {boneName=MSTR_hair_back_r}
    constraintSetup [{transformType=Rotation, targetBone {boneName=spine_05}}, {transformType=Rotation, targetBone {boneName=MSTR_hair_back_r}}]
    constraintWeights [1,1]
   }
   n10 = "Constraint - MSTR_hair_back_l (RR)"!(ComponentPose=@n9, Alpha=1, ConstraintWeights_0=0, ConstraintWeights_1:float=1) -> (Pose) {
    alphaBoolBlend {blendOption=Linear}
    boneToModify {boneName=MSTR_hair_back_l}
    constraintSetup [{transformType=Rotation, targetBone {boneName=spine_05}}, {transformType=Rotation, targetBone {boneName=MSTR_hair_back_l}}]
    constraintWeights [1,1]
   }
   n11 = "Transform (Modify) Bone - Bone: MSTR_hair_back_l"(ComponentPose=@n10, Alpha=@n7, Translation="-4,-3,0", Rotation="0,-50,0", Scale="1,1,1") {
    rotationMode=BMM_Additive
    rotationSpace=BCS_BoneSpace
    translationMode=BMM_Additive
    translationSpace=BCS_BoneSpace
    boneToModify {boneName=MSTR_hair_back_l}
    rotation {yaw=-50}
    translation=<-4 -3>
   }
   n12 = "Transform (Modify) Bone - Bone: MSTR_hair_mid_back"(ComponentPose=@n11, Alpha=@n7, Translation="-4,-3,0", Rotation="0,-50,0", Scale="1,1,1") {
    rotationMode=BMM_Additive
    rotationSpace=BCS_BoneSpace
    translationMode=BMM_Additive
    translationSpace=BCS_BoneSpace
    boneToModify {boneName=MSTR_hair_mid_back}
    rotation {yaw=-50}
    translation=<-4 -3>
   }
   n13 = "Transform (Modify) Bone - Bone: MSTR_hair_back_r"(ComponentPose=@n12, Alpha=@n7, Translation="-4,-3,0", Rotation="0,-50,0", Scale="1,1,1") {
    rotationMode=BMM_Additive
    rotationSpace=BCS_BoneSpace
    translationMode=BMM_Additive
    translationSpace=BCS_BoneSpace
    boneToModify {boneName=MSTR_hair_back_r}
    rotation {yaw=-50}
    translation=<-4 -3>
   }
   n14 = "Transform (Modify) Bone - Bone: upperarm_l"(ComponentPose=@n13, Alpha=1, Translation=@n6, Rotation=@n6.upperarm_l_t_Rotation, Scale="1,1,1") {
    rotationMode=BMM_Replace
    rotationSpace=BCS_ParentBoneSpace
    translationMode=BMM_Replace
    translationSpace=BCS_ParentBoneSpace
    boneToModify {boneName=upperarm_l}
   }
   n15 = Get!() -> (upperarm_r_t_Location:Vector, upperarm_r_t_Rotation:Rotator, upperarm_r_t_Scale:Vector)
   n16 = "Transform (Modify) Bone - Bone: upperarm_r"(ComponentPose=@n14, Alpha=1, Translation=@n15, Rotation=@n15.upperarm_r_t_Rotation, Scale="1,1,1") {
    rotationMode=BMM_Replace
    rotationSpace=BCS_ParentBoneSpace
    scaleSpace=BCS_ParentBoneSpace
    translationMode=BMM_Replace
    translationSpace=BCS_ParentBoneSpace
    boneToModify {boneName=upperarm_r}
   }
   n17 = Control Rig!(Source=Component To Local(ComponentPose=@n16), Alpha, has backpack:bool=Get has backpack()) -> (Pose) {controlRigAssetReference {controlRigAsset=&/F_MED_MistTycoonLife_FaceAcc_CtrlRig}}
   Save cached pose 'Pose'(Pose=@n17) {cacheName=Pose}
   n18 = Get Head Left() -> (Head Left:double)
   n19 = "Transform (Modify) Bone - Bone: MSTR_hair_back_l"(Alpha=@n18, Translation="0,0,0", Rotation="50,0,20", Scale="1,1,1") {
    rotationMode=BMM_Additive
    rotationSpace=BCS_BoneSpace
    translationMode=BMM_Additive
    translationSpace=BCS_BoneSpace
    boneToModify {boneName=MSTR_hair_back_l}
    rotation {pitch=50, roll=20}
   }
   n20 = "Transform (Modify) Bone - Bone: MSTR_hair_mid_back"(ComponentPose=@n19, Alpha=@n18, Translation="0,0,0", Rotation="50,0,50", Scale="1,1,1") {rotationMode=BMM_Additive, rotationSpace=BCS_BoneSpace, boneToModify {boneName=MSTR_hair_mid_back}, rotation {pitch=50, roll=50}}
   n21 = "Transform (Modify) Bone - Bone: MSTR_hair_back_r"(ComponentPose=@n20, Alpha=@n18, Translation="0,0,0", Rotation="50,0,0", Scale="1,1,1") {rotationMode=BMM_Additive, rotationSpace=BCS_BoneSpace, boneToModify {boneName=MSTR_hair_back_r}, rotation {pitch=50}}
   Reroute Node(InputPin:Component Space Pose Link Structure=@n21) -> (OutputPin:Component Space Pose Link Structure)
  ]}
 EventGraph {nodes [
   Event BlueprintUpdateAnimation()
   > n1 = Sequence!(execute) -> (then_0, then_1, then_2:Exec)
   > n2 = "Gravity Override State Function (Parameter Asset) (Nativized)"(
    CharacterPartAnimInstance=Self-Reference()
    GameTimeInSeconds=GetGameTimeInSeconds()
    Parameters=&/F_MED_F_MistTycoonLife_Params
   )
   > Set Gravity Override(Gravity Override=@n2)
   > Set RBAN Settings Out(RBAN Settings Out=@n2.OutRigidBodyAnimNodeParameters)
   Is Valid(exec=@n1.then_2, InputObject=Get body_skeletal_mesh() -> (body_skeletal_mesh:Skeletal Mesh Component Budgeted Object Reference))
   > Set upperarm_l_t(upperarm_l_t:Transform=GetBoneTransform!(
     self:Skinned Mesh Component Object Reference=Get body_skeletal_mesh()
     InBoneName:Name=upperarm_l
     TransformSpace:ERelativeTransformSpace Enum=RTS_ParentBoneSpace
    ) -> (ReturnValue:Transform)) -> (Output_Get:Transform)
   > Set upperarm_r_t(upperarm_r_t:Transform=GetBoneTransform(self=Get body_skeletal_mesh(), InBoneName=upperarm_r, TransformSpace=RTS_ParentBoneSpace)) -> (Output_Get:Transform)
   n3 = BreakTransform(InTransform=GetDeltaTransformFromRefPose(self=Get Skel Mesh(), BoneName=head, BaseName=spine_05))
   Set Head Up(execute=@n1.then_1, Head Up:double=MapRangeClamped(Value=BreakRotator(InRot=@n3.Rotation).Yaw, InRangeA=0, InRangeB=-50)) -> (Output_Get:double)
   n4 = Get Skel Mesh()
   n5 = BreakRotator(InRot=BreakTransform(InTransform=GetDeltaTransformFromRefPose(self=@n4, BoneName=head, BaseName=spine_05)).Rotation)
   > Set Head Right(Head Right:double=MapRangeClamped(Value=@n5, InRangeA=0, InRangeB=-50)) -> (Output_Get:double)
   > n6 = Branch(Condition=Get has backpack())
   > Set Head Right Hair(Head Right Hair:double=0) -> (Output_Get:double)
   > n7 = Set Head Right Hair Reverse(Head Right Hair Reverse:double=1) -> (Output_Get:double)
   n8 = Get Skel Mesh()
   n9 = BreakRotator(InRot=BreakTransform(InTransform=GetDeltaTransformFromRefPose(self=@n8, BoneName=head, BaseName=spine_05)).Rotation)
   Set Head Left(execute=[@n13.then, @n7.then], Head Left:double=MapRangeClamped(Value=@n9, InRangeA=0, InRangeB=50)) -> (Output_Get:double)
   > n10 = Branch(Condition=Get has backpack())
   > Set Head Left hair(Head Left hair:double=0) -> (Output_Get:double)
   > Set Head Left hair rev(Head Left hair rev:double=1) -> (Output_Get:double)
   n11 = MapRangeClamped(Value=@n9, InRangeA=30, InRangeB=70)
   Set Head Left hair(execute=@n10.else, Head Left hair=@n11)
   > Set Head Left hair rev(Head Left hair rev=MapRangeClamped(Value=@n11, InRangeA=0, InRangeB=1, OutRangeA=1, OutRangeB=0))
   n12 = MapRangeClamped(Value=@n5, InRangeA=-30, InRangeB=-70)
   Set Head Right Hair(execute=@n6.else, Head Right Hair=@n12)
   > n13 = Set Head Right Hair Reverse(Head Right Hair Reverse=MapRangeClamped(Value=@n12, InRangeA=0, InRangeB=1, OutRangeA=1, OutRangeB=0))
   n14 = Event BlueprintInitializeAnimation()
   n15 = Set Skel Mesh(execute=[@n14.then, @n16.then], Skel Mesh=GetOwnerSkeletalMeshComponent())
   n16 = Event OnCharacterCustomizationLoaded()
   Set body_skeletal_mesh(
    execute=@n15.then
    body_skeletal_mesh:Skeletal Mesh Component Budgeted Object Reference=Get Array Item with EFort Custom Part Type(Array=@n16.CustomMeshParts)
   ) -> (Output_Get:Skeletal Mesh Component Budgeted Object Reference)
   > n17 = Branch(Condition=CheckForBackbling())
   > Set has backpack(has backpack:bool=true) -> (Output_Get:bool)
   Set has backpack(execute=@n17.else, has backpack=false)
  ]}
]
variables [
 body_skeletal_mesh {type=SkeletalMeshComponentBudgeted}
 Gravity Override {type=Vector}
 has backpack {type=boolean}
 Head Left {type=number}
 Head Left hair {type=number}
 Head Left hair rev {type=number}
 Head Right {type=number}
 Head Right Hair {type=number}
 Head Right Hair Reverse {type=number}
 Head Up {type=number}
 RBAN Settings Out {type=FortRigidBodyAnimNodeParameters}
 Skel Mesh {type=SkeletalMeshComponent}
 upperarm_l_t {type=Transform, default {location=<0 0 0>, rotation=<rot 0 0 0>, scale=<1 1 1>}}
 upperarm_r_t {type=Transform, default {location=<0 0 0>, rotation=<rot 0 0 0>, scale=<1 1 1>}}
]
